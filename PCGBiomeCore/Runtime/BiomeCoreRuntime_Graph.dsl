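; Runtime hierarchical-generation PCG graph that scatters ground content and meshes per tile from data published by overlapping PCG_BiomeCore actors.
; Ground path samples biomes into points, pads empty neighbour cells with BiomeID 0, then loops GroundScatterRunGenerators with influence points, weighted assets, landscape and virtual textures on 64-size grids.
; Mesh path loads assembly data assets, uploads their points to GPU, copies them onto world transforms and spawns static meshes on 32-size grids.
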(graph BiomeCoreRuntime_Graph
  (params (DisplayBiomePoints :type Bool) (BiomesSearchDistance :type Double :default 3.2e+03) (InfluencePointsSearchDistance :type Double :default 1e+03) (MeshScattering :type Bool :default true) (GroundScattering :type Bool :default true) (UseHeightRVT :type Bool) (HeightRVT_Priming :type (Struct /Script/PCG.PCGVirtualTexturePrimingInfo) :default (struct :Grid Grid128)))
  (bind AttributeMathsOp_0 (Attribute%20Maths%20Op :Operation Multiply :InA InfluencePointsSearchDistance))
  (bind GetActorData_2 (Get%20Actor%20Data :ActorSelector (struct :ActorFilter AllWorldActors :bMustOverlapSelf true :ActorSelectionTag PCG_BiomeCore) :Mode GetActorReference))
  (bind GetActorData_0 (Get%20Actor%20Data :ActorSelector (struct :ActorFilter FromInput :bMustOverlapSelf true :ActorSelectionTag PCG_BiomeCore :ActorReferenceSelector ActorReference) :ExpectedPins [MeshScatter:AssetPaths MeshScatter:WorldTransforms GroundScatter:Biomes GroundScatter:InfluencePoints GroundScatter:WeightedAssets] :Mode GetDataFromPCGComponent))
  ;[0 0 0 0.5] Simple Generator (GPU)
  (bind GetLandscapeData_4 (Get%20Landscape%20Data :bComponentsMustOverlapSelf false :SamplingProperties (struct :bGetLayerWeights false)))
  (bind CopyAttributes_3 (Copy%20Attributes :bCopyAllAttributes true :InputSource @Last :Operation MergeSourcesAndCopyToAllTargets :Source UseHeightRVT :Source InfluencePointsSearchDistance :Target BiomesSearchDistance))
  (bind GetVirtualTextureData_0 (Get%20Virtual%20Texture%20Data))
  (bind Loop_0 (Loop :Graph /PCGBiomeCore/Runtime/GroundScatter/GroundScatterRunGenerators.GroundScatterRunGenerators))
  ;[0 0 0] Current Tile Influence Points
  (bind HiGenGridSize_6 (Change%20Grid%20Size :HiGenGridSize Grid64))
  (bind CullPointsOutsideActorBounds_92 (Cull%20Points%20Outside%20Actor%20Bounds :BoundsExpansion 2.5e+02))
  (bind CreatePoints_0 (Create%20Points :PointsToCreate []))
  (bind AddAttribute_0 (Add%20Attribute :AttributeTypes (struct :Type Integer32) :OutputTarget InfluenceType))
  (bind AddAttribute_1 (Add%20Attribute :AttributeTypes (struct :Type Float) :OutputTarget InfluenceRadius))
  (bind GetPointsCount_0 (Get%20Element%20Count))
  (bind AttributeReduce_0 (Attribute%20Reduce :Operation Sum))
  (bind Select_0 (Select))
  ;[0.15 0.15 0.15 0.5] Blend with no Biomes: Init with empty cells (BiomeID=0)
  (bind CreateConstant_6 (Create%20Constant :AttributeTypes (struct :DoubleValue 6.4e+03) :OutputTarget GridExtents))
  (bind MakeVectorAttribute_6 (Make%20Vector%20Attribute :OutputType Vector))
  (bind AttributeMathsOp_30 (Attribute%20Maths%20Op :InputSource1 GridExtents.xy :InB BiomesSearchDistance))
  (bind AttributeMathsOp_31 (Attribute%20Maths%20Op :InputSource1 GridExtents :Operation Multiply))
  (bind MakeVectorAttribute_7 (Make%20Vector%20Attribute :OutputType Vector :X BiomesSearchDistance :Y BiomesSearchDistance))
  (bind CreatePointsGrid_3 (Create%20Points%20Grid))
  (bind HiGenGridSize_11 (Change%20Grid%20Size :HiGenGridSize Grid64))
  (bind GetActorData_6 (Get%20Actor%20Data :Mode GetSinglePoint))
  (bind AttributeMathsOp_32 (Attribute%20Maths%20Op :InputSource1 @Last.xy :InputSource2 $Position.xy))
  (bind Difference_1 (Difference :bDiffMetadata false :DensityFunction Binary :Mode Discrete))
  (bind ToPoint_2 (To%20Point))
  (bind AddAttribute_3 (Add%20Attribute :AttributeTypes (struct :Type Integer32) :OutputTarget BiomeID))
  ;[0 0 0 0.5] Current Tile Biome Points
  (bind AttributeMathsOp_8 (Attribute%20Maths%20Op :Operation Subtract :InA BiomesSearchDistance))
  (bind HiGenGridSize_3 (Change%20Grid%20Size :HiGenGridSize Grid64))
  (bind DataNum_0 (Data%20Count))
  (bind Branch_7 (Branch))
  (bind CullPointsOutsideActorBounds_0 (Cull%20Points%20Outside%20Actor%20Bounds :BoundsExpansion 1.6e+03))
  (bind AttributePartition_2 (Attribute%20Partition :PartitionAttributeSelectors [$Position.xy]))
  (bind AttributeFilter_6 (Filter%20Attribute%20Elements :AttributeTypes (struct :Type Integer32) :bUseConstantThreshold true :Operator Equal :TargetAttribute $Index))
  (bind MergePoints_0 (Merge%20Points))
  (bind AttributeFilter_5 (Filter%20Attribute%20Elements :AttributeTypes (struct :Type Integer32 :DoubleValue -0.944) :bUseConstantThreshold true :Operator NotEqual :TargetAttribute BiomeID :ThresholdAttribute @Last))
  (bind GetPointsCount_1 (Get%20Element%20Count))
  (bind Branch_8 (Branch))
  (bind CopyAttributes_4 (Copy%20Attributes :InputSource $Position.xyz :OutputTarget @SourceName))
  (bind PointToAttributeSet_0 (Point%20To%20Attribute%20Set))
  (bind DataNum_1 (Data%20Count))
  (bind Branch_6 (Branch))
  (bind Subgraph_0 (Subgraph :Graph /PCG/Subgraphs/SpatialDataBoundsToPoint.SpatialDataBoundsToPoint))
  (bind Debug_0 (Debug))
  (bind Branch_1 (Branch :bOutputToB-param DisplayBiomePoints))
  (bind Debug_1 (Debug))
  ;[0.15 0.15 0.15 0.5] Mesh Scattering Inputs
  (bind Branch_3 (Branch :bOutputToB-param MeshScattering))
  (bind FilterDataByType_2 (Filter%20Data%20By%20Type :TargetType (struct :Ids [(struct :Struct /Script/PCG.PCGDataTypeInfoPoint)])))
  (bind Branch_2 (Branch :bOutputToB-param MeshScattering))
  (bind FilterDataByType_3 (Filter%20Data%20By%20Type :TargetType (struct :Ids [(struct :Struct /Script/PCG.PCGDataTypeInfoParam)])))
  ;[1 0.2259 0.3024 0.5] Mesh Scattering Assemblies
  (bind HiGenGridSize_4 (Change%20Grid%20Size :HiGenGridSize Grid32))
  (bind AttributePartition_4 (Attribute%20Partition :PartitionAttributeSelectors [RuntimeAssemblyHashValue]))
  (bind PCGLoadDataAsset_0 (Load%20PCG%20Data%20Asset :AssetReferenceSelector RuntimeAssembly :bLoadFromInput true :bWarnIfNoAsset false :DataIndexTag Data))
  (bind ToPoint_0 (To%20Point))
  (bind FilterDataByTag_0 (Filter%20Data%20By%20Tag :SelectedTags "Points"))
  (bind SortTags_0 (Sort%20Data%20By%20Tag%20Value :Tag Data))
  (bind DeleteAttributes_1 (Delete%20Attributes :Operation KeepSelectedAttributes :SelectedAttributes "Mesh,Material"))
  (bind CopyAttributes_2 (Copy%20Attributes :InputSource RuntimeAssemblyHashValue))
  (bind AttributeFilter_1 (Filter%20Attribute%20Elements :AttributeTypes (struct :Type SoftObjectPath) :bUseConstantThreshold true :Operator NotEqual :TargetAttribute Mesh))
  (bind CopyPoints_2 (Copy%20Points))
  (bind StaticMeshSpawner_2 (Static%20Mesh%20Spawner :bApplyMeshBoundsToPoints false :MeshSelectorParameters (class:/Script/PCG.PCGMeshSelectorByAttribute :AttributeName Mesh :TemplateDescriptor (struct :bAffectDynamicIndirectLighting false :bAffectDistanceFieldLighting false :bCastDynamicShadow false :bCastStaticShadow false :bVisibleInRayTracing false :bEvaluateWorldPositionOffset false :bIncludeInHLOD false :bForceNavigationObstacle false :WorldPositionOffsetDisableDistance 5000)) :MeshSelectorType /Script/PCG.PCGMeshSelectorByAttribute))
  ;;[0.15 0.15 0.15 0.5] Target Locations in the World (with Asset ID)
  (bind CullPointsOutsideActorBounds_3 (Cull%20Points%20Outside%20Actor%20Bounds))
  ;;[0.15 0.15 0.15 0.5] Upload on GPU
  (bind CustomHLSL_2 (Custom%20HLSL :OutputPins [(struct :PropertiesGPU (struct :PinsToInititalizeFrom [In] :CreatedKernelAttributeKeys [(struct :Identifier (struct :Name Mesh) :Type StringKey :Name Mesh)]) :Label Out :AllowedTypes (struct :Ids [(struct :Struct /Script/PCG.PCGDataTypeInfoPoint)]))]))
  ;[0 0 0] Debug Runtime Cells / Actors
  (bind Subgraph_2 (Subgraph :Graph /PCG/Subgraphs/SpatialDataBoundsToPoint.SpatialDataBoundsToPoint))
  (bind Debug_4 (Debug))
  ;[0.15 0.15 0.15 0.5] Ground Scattering Inputs
  (bind Branch_5 (Branch :bOutputToB-param GroundScattering))
  (bind FilterDataByType_0 (Filter%20Data%20By%20Type :TargetType (struct :Ids [(struct :Struct /Script/PCG.PCGDataTypeInfoPoint)])))
  (bind Branch_9 (Branch :bOutputToB-param GroundScattering))
  (bind FilterDataByType_5 (Filter%20Data%20By%20Type :TargetType (struct :Ids [(struct :Struct /Script/PCG.PCGDataTypeInfoParam)])))
  ;;[0.15 0.15 0.15 0.5] Sample Biomes
  (bind Branch_4 (Branch :bOutputToB-param GroundScattering))
  (bind DeleteAttributes_0 (Delete%20Attributes :Operation KeepSelectedAttributes))
  (bind FilterDataByType_7 (Filter%20Data%20By%20Type :TargetType (struct :Ids [(struct :Struct /Script/PCG.PCGDataTypeInfoSpatial)])))
  (bind GetActorData_7 (Get%20Actor%20Data :ActorSelector (struct :ActorFilter FromInput :bMustOverlapSelf true :ActorSelectionTag PCG_BiomeCore)))
  (bind MakeVectorAttribute_0 (Make%20Vector%20Attribute :OutputType Vector :X BiomesSearchDistance :Y BiomesSearchDistance))
  (bind VolumeSampler_0 (Volume%20Sampler :VoxelSize (1.6e+03 1.6e+03 3.2e+03)))
  (bind FilterDataByType_8 (Filter%20Data%20By%20Type :TargetType (struct :Ids [(struct :Struct /Script/PCG.PCGDataTypeInfoPoint)])))
  (bind GetAttributeFromPointIndex_0 (Get%20Attribute%20From%20Point%20Index :bForceOutputAttributeToBeInElementsDomain true :InputSource BiomeID))
  (bind CopyAttributes_1 (Copy%20Attributes :InputSource @Last))
  ;[0.15 0.15 0.15 0.5] Volume Height
  (bind GetActorData_5 (Get%20Actor%20Data :Mode GetSinglePoint))
  (bind CreateConstant_4 (Create%20Constant :OutputTarget VolumeSizeZ))
  (bind CopyAttributes_7 (Copy%20Attributes :InputSource $ScaledLocalSize.z :OutputTarget VolumeSizeZ))
  (reroute BiomePoints CopyAttributes_1)
  (reroute Influence%20Points FilterDataByType_0.InsideFilter)
  (reroute Mesh%20Scatter%20Locations FilterDataByType_2.InsideFilter)
  (reroute PCGDataAssetPaths FilterDataByType_3.InsideFilter)
  (reroute Volume%20Height CopyAttributes_7)
  (reroute WeightedAssets FilterDataByType_5.InsideFilter)
  (connect CreatePoints_0 Out AddAttribute_0 In)
  (connect AddAttribute_0 Out AddAttribute_1 In)
  (connect ToPoint_2 Out AddAttribute_3 In)
  (connect CopyAttributes_2 Out AttributeFilter_1 In)
  (connect MergePoints_0 Out AttributeFilter_5 In)
  (connect AttributePartition_2 Out AttributeFilter_6 In)
  (connect MakeVectorAttribute_6 Out AttributeMathsOp_30 InA)
  (connect AttributeMathsOp_30 Out AttributeMathsOp_31 InA)
  (connect CreatePointsGrid_3 Out AttributeMathsOp_32 InA)
  (connect GetActorData_6 Out AttributeMathsOp_32 InB)
  (connect AddAttribute_3 Out AttributePartition_2 In)
  (connect CullPointsOutsideActorBounds_0 Out AttributePartition_2 In)
  (connect PCGDataAssetPaths Out AttributePartition_4 In)
  (connect GetPointsCount_0 Out AttributeReduce_0 In)
  (connect Branch_8 Output%20B Branch_1 In)
  (connect GetActorData_0 MeshScatter:AssetPaths Branch_2 In)
  (connect GetActorData_0 MeshScatter:WorldTransforms Branch_3 In)
  (connect GetActorData_0 GroundScatter:Biomes Branch_4 In)
  (connect GetActorData_0 GroundScatter:InfluencePoints Branch_5 In)
  (connect DataNum_1 Out Branch_6 bOutputToB)
  (connect PointToAttributeSet_0 Out Branch_6 In)
  (connect DataNum_0 Out Branch_7 bOutputToB)
  (connect HiGenGridSize_3 Out Branch_7 In)
  (connect GetPointsCount_1 Out Branch_8 bOutputToB)
  (connect MergePoints_0 Out Branch_8 In)
  (connect GetActorData_0 GroundScatter:WeightedAssets Branch_9 In)
  (connect GetAttributeFromPointIndex_0 Attribute CopyAttributes_1 Source)
  (connect VolumeSampler_0 Out CopyAttributes_1 Target)
  (connect AttributePartition_4 Out CopyAttributes_2 Source)
  (connect DeleteAttributes_1 Out CopyAttributes_2 Target)
  (connect Branch_8 Output%20B CopyAttributes_4 Source)
  (connect Branch_8 Output%20B CopyAttributes_4 Target)
  (connect GetActorData_5 Out CopyAttributes_7 Source)
  (connect CreateConstant_4 Out CopyAttributes_7 Target)
  (connect CustomHLSL_2 Out CopyPoints_2 Source)
  (connect CullPointsOutsideActorBounds_3 Out CopyPoints_2 Target)
  (connect MakeVectorAttribute_7 Out CreatePointsGrid_3 CellSize)
  (connect AttributeMathsOp_31 Out CreatePointsGrid_3 GridExtents)
  (connect AttributeMathsOp_8 Out CullPointsOutsideActorBounds_0 BoundsExpansion)
  (connect Branch_7 Output%20B CullPointsOutsideActorBounds_0 In)
  (connect HiGenGridSize_4 Out CullPointsOutsideActorBounds_3 In)
  (connect AttributeMathsOp_0 Out CullPointsOutsideActorBounds_92 BoundsExpansion)
  (connect HiGenGridSize_6 Out CullPointsOutsideActorBounds_92 In)
  (connect AttributeFilter_1 InsideFilter CustomHLSL_2 In)
  (connect HiGenGridSize_3 Out DataNum_0 In)
  (connect PointToAttributeSet_0 Out DataNum_1 In)
  (connect Subgraph_0 Point Debug_0 In)
  (connect Branch_1 Output%20B Debug_1 In)
  (connect Subgraph_2 Point Debug_4 In)
  (connect Branch_4 Output%20B DeleteAttributes_0 In)
  (connect SortTags_0 Out DeleteAttributes_1 In)
  (connect CullPointsOutsideActorBounds_0 Out Difference_1 Differences)
  (connect AttributeMathsOp_32 Out Difference_1 Source)
  (connect ToPoint_0 Out FilterDataByTag_0 In)
  (connect Branch_5 Output%20B FilterDataByType_0 In)
  (connect Branch_3 Output%20B FilterDataByType_2 In)
  (connect Branch_2 Output%20B FilterDataByType_3 In)
  (connect Branch_9 Output%20B FilterDataByType_5 In)
  (connect DeleteAttributes_0 Out FilterDataByType_7 In)
  (connect Branch_4 Output%20B FilterDataByType_8 In)
  (connect GetActorData_2 Out GetActorData_0 In)
  (connect HiGenGridSize_11 Execution%20Dependency GetActorData_6 Execution%20Dependency)
  (connect GetActorData_2 Out GetActorData_7 In)
  (connect FilterDataByType_8 InsideFilter GetAttributeFromPointIndex_0 In)
  (connect CullPointsOutsideActorBounds_92 Out GetPointsCount_0 In)
  (connect AttributeFilter_5 InsideFilter GetPointsCount_1 In)
  (connect BiomePoints Out HiGenGridSize_3 In)
  (connect Mesh%20Scatter%20Locations Out HiGenGridSize_4 In)
  (connect Influence%20Points Out HiGenGridSize_6 In)
  (connect WeightedAssets Out Loop_0 AssetsByGeneratorGraph)
  (connect CopyAttributes_3 Out Loop_0 GraphParams)
  (connect GetLandscapeData_4 Out Loop_0 Landscape)
  (connect Branch_6 Output%20B Loop_0 TileBiomePoints)
  (connect Select_0 Out Loop_0 TileInfluencePoints)
  (connect GetVirtualTextureData_0 Out Loop_0 WorldHeightRVT)
  (connect Volume%20Height Out MakeVectorAttribute_0 Z)
  (connect CreateConstant_6 Out MakeVectorAttribute_6 X)
  (connect CreateConstant_6 Out MakeVectorAttribute_6 Y)
  (connect Volume%20Height Out MakeVectorAttribute_6 Z)
  (connect Volume%20Height Out MakeVectorAttribute_7 Z)
  (connect AttributeFilter_6 InsideFilter MergePoints_0 In)
  (connect PCGDataAssetPaths Out PCGLoadDataAsset_0 In)
  (connect CopyAttributes_4 Out PointToAttributeSet_0 In)
  (connect AttributeReduce_0 Out Select_0 bUseInputB)
  (connect AddAttribute_1 Out Select_0 Input%20A)
  (connect CullPointsOutsideActorBounds_92 Out Select_0 Input%20B)
  (connect FilterDataByTag_0 InsideFilter SortTags_0 In)
  (connect CopyPoints_2 Out StaticMeshSpawner_2 In)
  (connect HiGenGridSize_3 Grid%20Cell%20Volume Subgraph_0 SpatialData)
  (connect HiGenGridSize_4 Grid%20Cell%20Volume Subgraph_2 SpatialData)
  (connect PCGLoadDataAsset_0 Out ToPoint_0 In)
  (connect Difference_1 Out ToPoint_2 In)
  (connect GetActorData_7 Out VolumeSampler_0 Bounding%20Shape)
  (connect FilterDataByType_7 InsideFilter VolumeSampler_0 Volume)
  (connect MakeVectorAttribute_0 Out VolumeSampler_0 VoxelSize))
